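; Dynamic UI director component for DentureSay that shows a HUD action scene while the viewed pawn has the Mantis weapon equipped.
; Once UI is ready, it follows the local controller's view target changes, moving its weapon-equipped binding to each new player pawn and checking the current weapon.
; A cast to the Mantis weapon decides whether the scene is added or removed, only when its active state differs.
; EndPlay calls the view-target handler and unbinds from the controller.
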(event EventOnReadyToShowUI
  ; On Ready To Show UI
  (bind _asfort_player_controller_athena (Utilities|Casting|CastToFortPlayerControllerAthena (CallFunction|GetOwningLocalPlayerController)))
  (EventDispatchers|BindEventtoOnViewTargetChanged _asfort_player_controller_athena (EventDispatchers|CreateEvent :!event OnViewTargetChanged)))

(event EventEndPlay
  (params (EndPlayReason :type EEndPlayReason%20Enum :default "Destroyed"))
  ; End Play
  (CallFunction|OnViewTargetChanged 0)
  (bind _asfort_player_controller_athena (Utilities|Casting|CastToFortPlayerControllerAthena (CallFunction|GetOwningLocalPlayerController)))
  (EventDispatchers|UnbindEventfromOnViewTargetChanged _asfort_player_controller_athena (EventDispatchers|CreateEvent :!event OnViewTargetChanged)))

(fn HandleActionUIScene
  (params (Add :type Boolean))
  (if (!= (Variables|GetAdd) (DynamicUIDirector|IsSceneActive (CallFunction|GetOwningFortDynamicUIDirector) (Variables|Default|GetHudActionUIScene)))
    (CallFunction|AddorRemoveScene (Variables|Default|GetHudActionUIScene) (Variables|GetAdd))))

(fn OnViewTargetChanged
  (params (PlayerController :type Fort%20Player%20Controller%20Object%20Reference) (OldViewTarget :type Actor%20Object%20Reference) (NewViewTarget :type Actor%20Object%20Reference))
  (bind _outputdelegate (EventDispatchers|CreateEvent :!event OnWeaponEquipped))
  (bind (_updated_variable _previous_variable) (Utilities|FlowControl|DeltaGateWildcard (Variables|Default|GetCurrentViewTarget) (Utilities|Casting|CastToFortPlayerPawnAthena NewViewTarget))
    (:Success
      (LocalVariables|SetPreviousViewTarget _previous_variable)
      (Utilities|FlowControl|Sequence
        (:then_0
          (bind _previousviewtarget (LocalVariables|GetPreviousViewTarget)
            (:then
              (FortPawn|UnbindEventfromOnWeaponEquippedDelegate _previousviewtarget _outputdelegate))
            (:else)))
        (:then_1
          (bind _currentviewtarget (Variables|Default|GetCurrentViewTarget)
            (:then
              (FortPawn|BindEventtoOnWeaponEquippedDelegate _currentviewtarget _outputdelegate)
              (CallFunction|OnWeaponEquipped (Weapon|GetCurrentWeapon _currentviewtarget)))
            (:else)))))))

(fn OnWeaponEquipped
  (params (NewWeapon :type Fort%20Weapon%20Object%20Reference) (PrevWeapon :type Fort%20Weapon%20Object%20Reference))
  (Utilities|Casting|CastToB_DentureSay_Weapon_Mantis NewWeapon
    (:then
      (CallFunction|HandleActionUIScene true))
    (:CastFailed
      (CallFunction|HandleActionUIScene false))))
